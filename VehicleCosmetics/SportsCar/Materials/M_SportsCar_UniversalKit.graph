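# Default Lit material for the Sports Car Universal Kit vehicle cosmetic, enabled for skeletal meshes and mesh deformers.
# It holds three fixed texture samples from the Valet Sports Car kit: a _D color map, a normal map and an _S mask map.
# The graph has no parameters or math nodes and shows no links from the samples to material outputs.
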Material /VehicleCosmetics/SportsCar/Materials/M_SportsCar_UniversalKit
graph {
 properties {bUsedWithMeshDeformer=true, bUsedWithSkeletalMesh=true, pixelDepthOffsetMode=PDOM_Legacy, usedShadingModels=Default Lit}
 nodes [
  TextureSample() {texture=/VehicleCosmetics/SportsCar/Textures/T_Valet_Sports_Car_Universal_Kit_D}
  TextureSample() {samplerType=SAMPLERTYPE_Normal, texture=/VehicleCosmetics/SportsCar/Textures/T_Valet_Sports_Car_Universal_Kit_N}
  TextureSample() {samplerType=SAMPLERTYPE_Masks, texture=/VehicleCosmetics/SportsCar/Textures/T_Valet_Sports_Car_Universal_Kit_S}
 ]
}
