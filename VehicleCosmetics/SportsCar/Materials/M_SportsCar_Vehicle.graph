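# Clear coat surface material for the sports car vehicle cosmetic, usable on skeletal meshes and mesh deformers.
# The mask texture's red channel blends the diffuse texture toward a PrimaryColor parameter, which defaults to blue, allowing recolorable paint.
# The mask's alpha then blends that result toward a constant near-black value.
# Separate normal and packed S textures are sampled alongside a constant scalar.
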Material /VehicleCosmetics/SportsCar/Materials/M_SportsCar_Vehicle
graph {
 properties {
  bUsedWithMeshDeformer=true
  bUsedWithSkeletalMesh=true
  pixelDepthOffsetMode=PDOM_Legacy
  shadingModel=MSM_ClearCoat
  usedShadingModels=Clear Coat
 }
 nodes [
  Constant() {r=0.1}
  n1 = TextureSample() {samplerType=SAMPLERTYPE_Masks, texture=/VehicleCosmetics/SportsCar/Textures/T_Valet_Sports_Car_M}
  LinearInterpolate(
   A=LinearInterpolate(
    A=TextureSample() {texture=/VehicleCosmetics/SportsCar/Textures/T_Valet_Sports_Car_D}.RGB
    B=VectorParameter() {parameterName=PrimaryColor, defaultValue=<rgba 0.035 0.1288 0.3542 1>}.RGB
    Alpha=@n1.R
   )
   Alpha=@n1.A
  ) {constB=0.04}
  TextureSample() {samplerType=SAMPLERTYPE_Normal, texture=/VehicleCosmetics/SportsCar/Textures/T_Valet_Sports_Car_N}
  TextureSample() {samplerType=SAMPLERTYPE_Masks, texture=/VehicleCosmetics/SportsCar/Textures/T_Valet_Sports_Car_S}
 ]
}
